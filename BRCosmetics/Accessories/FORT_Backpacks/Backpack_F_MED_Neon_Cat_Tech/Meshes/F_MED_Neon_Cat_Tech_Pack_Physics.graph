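# Physics asset for the Neon Cat Tech backpack cosmetic, setting up collision and joints for its dangling belt.
# Three capsules cover the pack body, and single capsules cover the master belt bone and four chained belt bones.
# The master belt rotates freely on the pack, while the first segment swings mostly on one axis.
# Later segments allow moderate swing on both axes, with twist kept very tight along the chain.
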PhysicsAsset /BRCosmetics/Accessories/FORT_Backpacks/Backpack_F_MED_Neon_Cat_Tech/Meshes/F_MED_Neon_Cat_Tech_Pack_Physics
bodies {
 Dyn_Belt_01 [{shapeType=Capsule, length=5, radius=0.75, shapeName=Dyn_Belt_01_capsule, center=<3 0.6277 0>, rotation=<rot 90 0 0>}]
 Dyn_Belt_02 [{shapeType=Capsule, length=5, radius=0.75, shapeName=Dyn_Belt_02_capsule, center=<3 0 0>, rotation=<rot 90 90 90.4138>}]
 Dyn_Belt_03 [{shapeType=Capsule, length=5, radius=1, shapeName=Dyn_Belt_03_capsule, center=<3 0 0>, rotation=<rot 90 90 90.4138>}]
 Dyn_Belt_04 [{shapeType=Capsule, length=5, radius=1, shapeName=Dyn_Belt_04_capsule, center=<3 0 0>, rotation=<rot 90 0 0>}]
 Dyn_Master_Belt [{shapeType=Capsule, length=0.505, radius=0.505, shapeName=Dyn_Master_Belt_capsule, rotation=<rot 90 0 0>}]
 pack [
  {shapeType=Capsule, length=20, radius=8, center=<-0.7551 -9.5942 -6.4872>, rotation=<rot -34.0904 0 0>}
  {shapeType=Capsule, length=6.75, radius=2, shapeName=pack_capsule, center=<4.2052 -4.9146 -11.7655>, rotation=<rot -45.9576 0 0>}
  {shapeType=Capsule, length=6, radius=5.5, shapeName=pack_capsule, center=<4.7541 4.9991 -22.0451>}
 ]
}
constraints [
 {
  bone1Name=Dyn_Master_Belt
  bone2Name=pack
  swing1LimitDegrees=45
  swing1Motion=Free
  swing2LimitDegrees=45
  swing2Motion=Free
  twistLimitDegrees=45
  twistMotion=Free
 }
 {bone1Name=Dyn_Belt_01, bone2Name=Dyn_Master_Belt, swing1LimitDegrees=40, swing2LimitDegrees=1, twistLimitDegrees=1}
 {bone1Name=Dyn_Belt_02, bone2Name=Dyn_Belt_01, swing1LimitDegrees=20, swing2LimitDegrees=20, twistLimitDegrees=1}
 {bone1Name=Dyn_Belt_03, bone2Name=Dyn_Belt_02, swing1LimitDegrees=20, swing2LimitDegrees=20, twistLimitDegrees=1}
 {bone1Name=Dyn_Belt_04, bone2Name=Dyn_Belt_03, swing1LimitDegrees=20, swing2LimitDegrees=20, twistLimitDegrees=1}
]
